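# Conversation database for the RainbowStraps island NPC, defining its greeting dialogue and the options that follow.
# The entry tag leads to one speech node that picks a contextual line by priority: first-ever meeting, a reaction to a specific player outfit, or match-conversation-count-gated chatter, with a default fallback.
# The speech then branches through link nodes to the shared NPC library quests, basic services and a Ch5S1 quest conversation.
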ConversationDatabase /NPCLibrary/NPCs/RainbowStraps/NPCLibrary_Conversation_RainbowStraps
entry_points [{tag=Conversation.Entry.RainbowStraps, destinations [#1]}]
nodes [
 ConversationEntryPointNode_7 {
  id=#1
  class=ConversationEntryPointNode
  outputs [#2]
  properties {entryTag {tagName=Conversation.Entry.RainbowStraps}}
 }
 Speech_C_4 {
  id=#2
  class=Speech_C
  outputs [#3, #4, #5]
  properties {
   nodeName=""
   generalConfig {
    defaultMessage="Who's your favorite person on the island? It's okay, I'll wait."
    contextualMessages [
     {
      message=Have you seen any vending machines around? Just curious.
      contextRequirements [{requirement {class=FortControllerRequirement_IsFirstEverConversationWithNPC}}]
      priority {curve {rowName=Tandem.Speech.Priority.FirstEverInteraction}}
     }
     {
      message=I LOVE your style.
      contextRequirements [{requirement {
         class=FortControllerRequirement_HasCID
         properties {allowedCIds [/BRCosmetics/Athena/Items/Cosmetics/Characters/CID_A_029_Athena_Commando_M_LlamaHeroWinter_C83TZ]}
        }}]
      priority {curve {rowName=Tandem.Speech.Priority.OutfitReaction}}
     }
     {
      message="Who's your favorite person on the island? It's okay, I'll wait."
      contextRequirements [{requirement {
         class=FortControllerRequirement_CheckNPCMatchConversationCount
         properties {nPCUniqueIdTag {tagName=AISpawnerData.Type.Tandem.RainbowStraps}}
        }}]
      priority {curve {rowName=None}}
     }
     {
      message=What's MY favorite weapon? Can't I just say all of them?
      contextRequirements [{requirement {
         class=FortControllerRequirement_CheckNPCMatchConversationCount
         properties {nPCUniqueIdTag {tagName=AISpawnerData.Type.Tandem.RainbowStraps}}
        }}]
      priority {curve {rowName=None}}
     }
     {
      message=What's a Llama got to do to get a weapon around here?
      contextRequirements [{requirement {
         class=FortControllerRequirement_CheckNPCMatchConversationCount
         properties {nPCUniqueIdTag {tagName=AISpawnerData.Type.Tandem.RainbowStraps}}
        }}]
      priority {curve {rowName=None}}
     }
     {
      message="Can I call you bestie, or is that like, so cringe?"
      contextRequirements [{requirement {
         class=FortControllerRequirement_CheckNPCMatchConversationCount
         properties {nPCUniqueIdTag {tagName=AISpawnerData.Type.Tandem.RainbowStraps}}
        }}]
      priority {curve {rowName=None}}
     }
     {
      message=How's the battle going? Need a pep talk?
      contextRequirements [{requirement {
         class=FortControllerRequirement_CheckNPCMatchConversationCount
         properties {nPCUniqueIdTag {tagName=AISpawnerData.Type.Tandem.RainbowStraps}}
        }}]
      priority {curve {rowName=None}}
     }
    ]
   }
  }
 }
 ConversationLinkNode_2 {id=#5, class=ConversationLinkNode, properties {remoteEntryTag {tagName=Conversation.Entry.Ch5S1.D3.Q8}}}
 ConversationLinkNode_0 {id=#4, class=ConversationLinkNode, properties {remoteEntryTag {tagName=Conversation.Entry.BasicServices}}}
 ConversationLinkNode_1 {id=#3, class=ConversationLinkNode, properties {remoteEntryTag {tagName=Conversation.Entry.NPCLibrary.Quests}}}
]
